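# Niagara impact burst for the CherryKnot pickaxe, combining layered flashes, smoke, toon sparks and flipbook stickers.
# The flash emitter spawns four groups, with switches on spawn group and visibility tag giving each its own look across five sprite renderers.
# Smoke and stickers burst outward against drag; stickers spin, fall and fake flipping through sine-scaled width.
# System values supply front-end scale and offset, plus time-of-day tint and brightness for selected layers.
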NiagaraSystem /BRCosmetics/Weapons/FORT_Melee/Pickaxe_CherryKnot/FX/NS_Pickaxe_CherryKnot_Impact
dependencies {
 usedDataInterfaces [
  &NiagaraDataInterfaceCurve
  &NiagaraDataInterfaceColorCurve
  &NiagaraDataInterfaceVector2DCurve
  &NiagaraDataInterfaceSpriteRendererInfo
 ]
 usedDynamicInputs [
  &FN_FrontEndFloatModifier
  &FN_FrontEndVector3Modifier
 ]
 usedModules [
  &self:SystemSpawnScript.SystemScriptSource.SystemScriptGraph.NiagaraNodeAssignment_0.SetVariables_#1
  &self:SystemSpawnScript.SystemScriptSource.SystemScriptGraph.NiagaraNodeAssignment_1.SetVariables_#2
  &self:SystemSpawnScript.SystemScriptSource.SystemScriptGraph.NiagaraNodeAssignment_2.SetVariables_#3
  &SystemState
 ]
}
emitters {
 Flash_Elements001 {
  emitterUpdateScript [
   &EmitterState(
    Enable Distance Culling=true
    Max Distance Response=::KillImmediately
    Module.bUseMaxDistance=true
    Module.bUseMinDistance=false
    Scalability Mode=::NewEnumerator1
   )
   &SpawnBurst_Instantaneous(Spawn Count=1, Spawn Group=3, Use Spawn Probability=false) {module=SpawnBurst_Instantaneous003}
   &SpawnBurst_Instantaneous(Spawn Count=1, Spawn Group=2, Use Spawn Probability=false) {module=SpawnBurst_Instantaneous002}
   &SpawnBurst_Instantaneous(Spawn Group=1, Use Spawn Probability=false, UseLoopCountLimit=false) {module=SpawnBurst_Instantaneous001}
   &SpawnBurst_Instantaneous(Use Spawn Probability=false, UseLoopCountLimit=false)
  ]
  particleSpawnScript [
   &InitializeParticle(
    Color=&MultiplyLinearColor(
     Linear Color=&MergeHelper_ColorSwitch(0=<rgba 1 1 1 1>, 1=<rgba 1 1 1 0.1>, 2=<rgba 1 1 1 1>, 3=<rgba 1 1 1 1>, Index=$Emitter.SpawnGroup, Num=::NewEnumerator2)
     Scale Factor=&MakeVector4_FromLinearColor(LinearColor=&LerpLinearColors(
       StartColor=<rgba 1 1 1 1>
       EndColor=&Multiply_LinearColorByFloat(
        Linear Color=$System.TOD
        Float=&FNTimeOfDayScalar(Night Scale=2, Morning Scale=1.5, Evening Scale=2)
       )
       LerpFactor=0.6
      ))
    )
    Lifetime=&MergeHelper_FloatSwitch(0=0.133, 1=0.32, 2=0.2, 3=0.3, Index=$Emitter.SpawnGroup, Num=::NewEnumerator2)
    Position Offset=$System.FEImpactOffset
    Sprite Rotation Angle=&MergeHelper_FloatSwitch(
     0=&RandomRangeFloat(Minimum=0, Maximum=1)
     1=&RandomRangeFloat(Minimum=-0.125, Maximum=0.125)
     2=&RandomRangeFloat(Minimum=0, Maximum=1)
     3=&RandomRangeFloat(Minimum=0, Maximum=1)
     Index=$Emitter.SpawnGroup
     Num=::NewEnumerator2
    )
    Sprite Rotation Mode=::NewEnumerator1
    Sprite Size Mode=::NewEnumerator3
    Uniform Sprite Size=&Multiply_Float(
     A=&MergeHelper_FloatSwitch(
      0=&UniformRangedFloat(Minimum=100, Maximum=120)
      1=&RandomRangeFloat(Minimum=100, Maximum=120)
      2=200
      3=&Multiply_Float(A=140, B=&RandomRangeFloat(Minimum=0.7, Maximum=1))
      Index=$Emitter.SpawnGroup
      Num=::NewEnumerator2
     )
     B=$System.FrontEndScale
    )
    UsePositionOffset=true
    Write Material Random=true
   )
   &CameraOffset(Camera Offset Amount=&MergeHelper_FloatSwitch(0=10, 1=30, 2=25, 3=10, Index=$Emitter.SpawnGroup, Num=::NewEnumerator2))
   /SetVariables(Particles.VisibilityTag=$Emitter.SpawnGroup) {isSetParameters=true}
   /SetVariables() {isSetParameters=true}
   /SetVariables(Particles.DynamicParameterAlt2:Vector4f=&MakeVector4(X=1, Y=1, Z=0, W=0)) {isSetParameters=true}
  ]
  particleUpdateScript [
   &ParticleState()
   &ScaleSpriteSize(
    Scale Sprite Size Mode=::NewEnumerator0
    Uniform Scale Factor=&MergeHelper_FloatSwitch(
     0=&FloatFromCurve(FloatCurve=DataInterface() {propertyValues {Curve {keys [
          {arriveTangent=2.105, interpMode=RCIM_Cubic, leaveTangent=2.105, tangentMode=RCTM_User, time=0, value=0}
          {arriveTangent=0.0224, interpMode=RCIM_Cubic, leaveTangent=0.0224, tangentMode=RCTM_User, time=1, value=1}
         ]}}})
     1=1
     2=&FloatFromCurve(FloatCurve=DataInterface() {propertyValues {Curve {keys [
          {arriveTangent=1.6131, interpMode=RCIM_Cubic, leaveTangent=1.6131, tangentMode=RCTM_User, time=0, value=0.7}
          {arriveTangent=0.0224, interpMode=RCIM_Cubic, leaveTangent=0.0224, tangentMode=RCTM_User, time=1, value=0}
         ]}}})
     3=&FloatFromCurve(FloatCurve=DataInterface() {propertyValues {Curve {keys [{arriveTangent=3.9086, interpMode=RCIM_Cubic, leaveTangent=3.9086, tangentMode=RCTM_User, time=0, value=0}, <at 1 1.4 RCIM_Cubic>]}}})
     Index=$Particles.VisibilityTag
     Num=::NewEnumerator2
    )
   ) {module=ScaleSpriteSize001}
   &ScaleColor(
    Scale Alpha=&MergeHelper_FloatSwitch(
     0=&FloatFromCurve(
      FloatCurve=DataInterface() {propertyValues {Curve {keys [
          {arriveTangent=9.0725, interpMode=RCIM_Cubic, leaveTangent=9.0725, tangentMode=RCTM_User, time=0, value=0}
          {arriveTangent=0.3975, interpMode=RCIM_Cubic, leaveTangent=0.3975, tangentMode=RCTM_User, time=0.0555, value=1}
          {arriveTangent=-0.0405, interpMode=RCIM_Cubic, leaveTangent=-0.0405, tangentMode=RCTM_User, time=1, value=0}
         ]}}}
      Scale Curve=$System.ToDScalar
     )
     1=&FloatFromCurve(
      FloatCurve=DataInterface() {propertyValues {Curve {keys [<at 0 1>,<at 0.8091 0.9932>,<at 1 0>]}}}
      Scale Curve=&FNTimeOfDayScalar(Night Scale=5, Morning Scale=4, Day Scale=2, Evening Scale=4)
     )
     2=$System.ToDScalar
     3=&FloatFromCurve(
      FloatCurve=DataInterface() {propertyValues {Curve {keys [
          {arriveTangent=-0.2921, interpMode=RCIM_Cubic, leaveTangent=-0.2921, tangentMode=RCTM_User, time=0, value=7}
          {arriveTangent=-7, interpMode=RCIM_Cubic, leaveTangent=-7, time=0.4, value=4}
          <at 1 0 RCIM_Cubic>
         ]}}}
      Scale Curve=2
     )
     Index=$Particles.VisibilityTag
     Num=::NewEnumerator2
    )
    Scale RGB=&MergeHelper_VectorSwitch(
     0=&VectorFromFloat(Value=&FloatFromCurve(
       FloatCurve=DataInterface() {propertyValues {Curve {keys [<at 0 5 RCIM_Cubic>, {arriveTangent=-4, interpMode=RCIM_Cubic, leaveTangent=-4, time=0.5, value=2}, {tangentMode=RCTM_User, time=1, value=1}]}}}
       Scale Curve=$System.ToDScalar
      ))
     1=&MakeVectorFromLinearColorRGB(Color=&FNTimeOfDayColor(
       Night Color=<rgba 0.34 0.51 0.85 1>
       Morning Color=<rgba 0.902 0.5882 0.5412 1>
       Day Color=<rgba 1 1 1 1>
       Evening Color=<rgba 0.9 0.54 0.6 1>
      ))
     2=&MakeVectorFromLinearColorRGB(Color=&ColorFromCurve(ColorCurve=DataInterface() {propertyValues {
         AlphaCurve {keys [<at 0 1 RCIM_Cubic>]}
         BlueCurve {keys [<at 0 0.4972>,<at 0.7773 0.2526>]}
         GreenCurve {keys [<at 0 0.6615>,<at 0.7773 0.3351>]}
         RedCurve {keys [<at 0 0.7344>,<at 0.7773 0.5>]}
        }}))
     3=<1 1 1>
     Index=$Particles.VisibilityTag
     Num=::NewEnumerator2
    )
    ScaleA=true
    ScaleRGB=true
   )
   /SetVariables(
    Particles.InnerColor:LinearColor=&MultiplyLinearColor(
     Linear Color=<rgba 0.5833 0.5833 0.5833 1>
     Scale Factor=&MakeVector4_FromLinearColor(LinearColor=$Particles.Color)
    )
    Particles.InnerDynamicParameter:Vector4f=&MakeVector4_FromFloat(Float=&Lerp_Float(A=0, B=0.5, Alpha=$Particles.NormalizedAge))
    Particles.InnerSize:Vector2f=&Multiply_Vector2D_ByFloat(Float=0.5, Vector2D=$Particles.SpriteSize)
    Particles.MiddleColor:LinearColor=&MultiplyLinearColor(
     Linear Color=<rgba 0.4479 0.4479 0.4479 1>
     Scale Factor=&MakeVector4_FromLinearColor(LinearColor=$Particles.Color)
    )
    Particles.MiddleDynamicParameter:Vector4f=&MakeVector4_FromFloat(Float=&Lerp_Float(A=0.25, B=0.75, Alpha=$Particles.NormalizedAge))
    Particles.MiddleSize:Vector2f=&Multiply_Vector2D_ByFloat(Float=0.75, Vector2D=$Particles.SpriteSize)
    Particles.OuterColor:LinearColor=&MultiplyLinearColor(
     Linear Color=<rgba 0.1354 0.1354 0.1354 1>
     Scale Factor=&MakeVector4_FromLinearColor(LinearColor=$Particles.Color)
    )
    Particles.OuterDynamicParameter:Vector4f=&MakeVector4_FromFloat(Float=&Lerp_Float(A=0, B=1, Alpha=$Particles.NormalizedAge))
    Particles.OuterSize:Vector2f=&Multiply_Vector2D_ByFloat(Float=1.3, Vector2D=$Particles.SpriteSize)
    Particles.SubImageIndex=&RandomRangeFloat(Minimum=0, Maximum=4)
    Particles.SubImageIndexAlt:NiagaraFloat=&RandomRangeFloat(Minimum=0, Maximum=3)
   ) {isSetParameters=true}
   /SetVariables(Particles.DynamicParameterAlt1:Vector4f=&MakeVector4(
     X=&FloatFromCurve(FloatCurve=DataInterface() {propertyValues {Curve {keys [<at 0 0>,<at 1 1>]}}})
     Y=&UniformAOrBFloat(A=0, Evaluation Type=::NewEnumerator0, Recalculate Random Each Loop=true, Override Seed=false)
     Z=0
     W=0
    )) {isSetParameters=true}
   /SetVariables(Particles.DynamicParameterAlt3:Vector4f=&MakeVector4(
     X=&FloatFromCurve(FloatCurve=DataInterface() {propertyValues {Curve {keys [<at 0 1 RCIM_Cubic>, {arriveTangent=-2.742, interpMode=RCIM_Cubic, leaveTangent=-2.742, tangentMode=RCTM_User, time=1, value=0}]}}})
     Y=&FloatFromCurve(FloatCurve=DataInterface() {propertyValues {Curve {keys [<at 0 0 RCIM_Cubic>, {arriveTangent=2.4833, interpMode=RCIM_Cubic, leaveTangent=2.4833, tangentMode=RCTM_User, time=1, value=1}]}}})
     Z=0
     W=0
    )) {isSetParameters=true}
  ]
  rendererClasses [&NiagaraSpriteRendererProperties]
  renderers [
   {rendererClass=&NiagaraSpriteRendererProperties}
   {rendererClass=&NiagaraSpriteRendererProperties, rendererIndex=1}
   {rendererClass=&NiagaraSpriteRendererProperties, rendererIndex=2}
   {rendererClass=&NiagaraSpriteRendererProperties, rendererIndex=3}
   {rendererClass=&NiagaraSpriteRendererProperties, rendererIndex=4}
  ]
 }
 Impact_Smoke {
  emitterUpdateScript [
   &EmitterState()
   &SpawnBurst_Instantaneous(
    Spawn Count=&RandomRangeInt(Minimum=15, Maximum=20)
    Use Spawn Probability=false
    UseLoopCountLimit=false
   )
  ]
  particleSpawnScript [
   &InitializeParticle(
    Color=&MultiplyLinearColor(
     Linear Color=&RandomRangeLinearColor(Minimum=<rgba 0.7417 0.8501 0.9271 0.5>, Maximum=<rgba 0.9479 0.8471 0.7603 0.8>)
     Scale Factor=&MakeVector4_FromLinearColor(LinearColor=&LerpLinearColors(StartColor=<rgba 1 1 1 1>, EndColor=$System.TOD, LerpFactor=0.8))
    )
    Lifetime=&FloatFromCurve(
     FloatCurve=DataInterface() {propertyValues {Curve {keys [<at 0 0.3>,<at 0.8 0.4>,<at 1 0.6>]}}}
     CurveIndex=&RandomRangeFloat(Minimum=0, Maximum=1)
    )
    Position Offset=$System.FEImpactOffset
    Sprite Rotation Mode=::NewEnumerator2
    Sprite Size Mode=::NewEnumerator3
    Uniform Sprite Size=&Multiply_Float(A=$System.FrontEndScale, B=&RandomRangeFloat(Minimum=30, Maximum=45))
    UsePositionOffset=true
    Write Material Random=true
   )
   &CameraOffset(Camera Offset Amount=&RandomRangeFloat(Minimum=40, Maximum=50, Evaluation Type=::NewEnumerator1))
   &ShapeLocation(
    Hemisphere Distribution=<1 0.5>
    MaskSpawnGroup=false
    Override Seed=false
    Rotation Angle=90
    Rotation Mode=::NewEnumerator1
    Sphere Radius=&Multiply_Float(A=$System.FrontEndScale, B=10)
    UseOwnerScale=true
   )
   &AddVelocity(
    Velocity=&Multiply_VectorByFloat(Vector=$Particles.ShapeLocation.ShapeNormal, Float=&RandomRangeFloat(Minimum=280, Maximum=340))
    Velocity Speed Scale=$System.FrontEndScale
   )
  ]
  particleUpdateScript [
   &ParticleState()
   &ScaleSpriteSize(Uniform Curve Sprite Scale=DataInterface() {propertyValues {Curve {keys [{arriveTangent=3.0522, interpMode=RCIM_Cubic, leaveTangent=3.0522, tangentMode=RCTM_User, time=0, value=0.3}, <at 1 1 RCIM_Cubic>]}}})
   &ScaleColor(
    Scale Alpha=&FloatFromCurve(FloatCurve=DataInterface() {propertyValues {Curve {keys [<at 0 1 RCIM_Cubic>,<at 1 0 RCIM_Cubic>]}}}, Scale Curve=0.5)
    Scale RGB=&MakeVectorFromLinearColorRGB(Color=&ColorFromCurve(ColorCurve=DataInterface() {propertyValues {
        AlphaCurve {keys [<at 0 1 RCIM_Cubic>]}
        BlueCurve {keys [<at 0 1 RCIM_Cubic>,<at 1 0.2656 RCIM_Cubic>]}
        GreenCurve {keys [<at 0 1 RCIM_Cubic>,<at 1 0.2656 RCIM_Cubic>]}
        RedCurve {keys [<at 0 1 RCIM_Cubic>,<at 1 0.2656 RCIM_Cubic>]}
       }}))
    ScaleA=true
    ScaleRGB=true
   )
   &DynamicMaterialParameters(
    Index 0 Param 1=&FloatFromCurve(FloatCurve=DataInterface() {propertyValues {Curve {keys [{arriveTangent=-1.1894, interpMode=RCIM_Cubic, leaveTangent=-1.1894, tangentMode=RCTM_User, time=0, value=1}, <at 1 0 RCIM_Cubic>]}}})
    Module.Param0WriteEnabled=true
    Module.Param1WriteEnabled=true
    Module.Param2WriteEnabled=true
    Module.Param3WriteEnabled=true
   ) {module=DynamicMaterialParameters001}
   &Drag(Drag=&RandomRangeFloat(Minimum=8, Maximum=9), Use Linear Drag=true, Use Rotational Drag=true)
   &SolveForcesAndVelocity(Clamp Velocity=false, Limit Acceleration=false)
   &SubUVAnimation(
    Direct Frame Index:NiagaraFloat=/Niagara/DynamicInputs/Integer/V2/Round_Float(Float:NiagaraFloat=&Add_Float(
      A=&RandomRangeFloat(Minimum=0, Maximum=40)
      B=&Lerp_Float(
       A=0
       B=&RandomRangeFloat(Minimum=15, Maximum=20)
       Alpha=&FloatFromCurve(FloatCurve=DataInterface() {propertyValues {Curve {defaultValue=1, keys [{tangentMode=RCTM_User, time=0, value=0}, {tangentMode=RCTM_User, time=1, value=1}]}}})
      )
     ))
    "Sprite Renderer "=DataInterface() {propertyValues {SpriteRenderer=&self:Impact_Smoke_1.NiagaraSpriteRendererProperties_2}}
    SubUV Animation Mode:ENiagara_SubUVLookupModeV2=::NewEnumerator5
   )
  ]
  rendererClasses [&NiagaraSpriteRendererProperties]
  renderers [{rendererClass=&NiagaraSpriteRendererProperties}]
 }
 Stickers {
  emitterUpdateScript [
   &EmitterState(Life Cycle Mode=::NewEnumerator1, Loop Behavior=::NewEnumerator1, Loop Duration=1, UseLoopDelay=false)
   &SpawnBurst_Instantaneous(
    Spawn Count=&MakeIntFromFloat(Float=&Multiply_FloatByInt(Float=$System.FrontEndScale, Integer=&RandomRangeInt(Minimum=4, Maximum=6)))
    Use Spawn Probability=false
    UseLoopCountLimit=false
   )
  ]
  particleSpawnScript [
   &FN_SetSpawnGroup(Number of Spawn Groups=6)
   &PartitionParticles(Number Of Partitions=6)
   /SetVariables(
    Particles.CustomRandom:NiagaraFloat=&RandomRangeFloat(Minimum=0, Maximum=1)
    Particles.CustomRandomBool:NiagaraBool=&RandomBool()
   ) {isSetParameters=true}
   &InitializeParticle(
    Lifetime=&RandomRangeFloat(Minimum=0.3, Maximum=0.8)
    Position Offset=$System.FEImpactOffset
    Sprite Rotation Mode=::NewEnumerator2
    Sprite Size Mode=::NewEnumerator3
    Uniform Sprite Size=&Multiply_Float(A=$System.FrontEndScale, B=&RandomRangeFloat(Minimum=4, Maximum=7))
    UsePositionOffset=true
    Write Material Random=true
   )
   /SetVariables(Particles.InitialSize:Vector2f=&Vector2DFromFloat(Value=&RandomRangeFloat(Minimum=50, Maximum=80))) {isSetParameters=true}
   &CameraOffset(Camera Offset Amount=10)
   &SubUVAnimation(
    Direct Frame Index:NiagaraFloat=&Multiply_FloatByInt(Float=1, Integer=$Particles.PartitionParticles.Partition)
    "Sprite Renderer "=DataInterface() {propertyValues {SpriteRenderer=&self:Stickers_0.NiagaraSpriteRendererProperties_2}}
    SubUV Animation Mode:ENiagara_SubUVLookupModeV2=::NewEnumerator5
   )
   &ShapeLocation(
    Additional Quaternion Rotation=true
    Additional Yaw / Pitch / Roll=<-90 0 0>
    Hemisphere Distribution=<0.5 1>
    MaskSpawnGroup=false
    Override Seed=false
    Rotation Angle Type:ENiagara_AngleInput=::NewEnumerator0
    Rotation Mode=::NewEnumerator3
    Sphere Radius=&Multiply_Float(A=10, B=$System.FrontEndScale)
    UseOwnerScale=true
   )
   &AddVelocity(
    Cone Angle=30
    Cone Axis=$Particles.ShapeLocation.ShapeNormal
    Module.UseConeFalloff=false
    Module.UseVelDistribution=false
    Override Seed=false
    Rotation Coordinate Space=::Simulation
    Velocity Mode=::NewEnumerator2
    Velocity Speed=&Multiply_Float(A=&RandomRangeFloat(Minimum=100, Maximum=500), B=$System.FrontEndScale)
   ) {module=AddVelocity002}
  ]
  particleUpdateScript [
   &ParticleState()
   &ScaleColor(
    Scale RGB=&Multiply_Vector(
     A=<1 1 1>
     B=&VectorFromFloat(Value=&FloatFromCurve(FloatCurve=DataInterface() {propertyValues {Curve {keys [<at 0 0>,<at 0.1 1>,<at 1 1>]}}}))
    )
    ScaleA=true
    ScaleRGB=true
   )
   &ScaleSpriteSize(Scale Factor=&Vector2DFromFloat(Value=&FloatFromCurve(
      FloatCurve=DataInterface() {propertyValues {Curve {keys [
          {arriveTangent=21.6697, interpMode=RCIM_Cubic, leaveTangent=21.6697, tangentMode=RCTM_User, time=0, value=0}
          {arriveTangent=-0.0042, leaveTangent=-0.0042, time=0.05, value=1}
          {arriveTangent=-1.0526, interpMode=RCIM_Cubic, leaveTangent=-1.0526, time=0.4, value=1}
          {arriveTangent=-0.6473, interpMode=RCIM_Cubic, leaveTangent=-0.6473, tangentMode=RCTM_User, time=1, value=0}
         ]}}}
      Scale Curve=2
     ))) {module=ScaleSpriteSize001}
   &SpriteRotationRate(Rotation Rate=&RandomRangeFloat(Minimum=150, Maximum=220))
   &ScaleSpriteSize(Scale Factor=&Lerp_Vector2(
     A=&Multiply_Vector2(
      A=&MakeVector2D(
       X=&Multiply_Float(
        A=1
        B=&Sine(
         Normalized Angle=$Emitter.Age
         Period=&Multiply_Float(
          A=&RandomRangeFloat(Minimum=1.2, Maximum=1.8)
          B=&Add_Float(A=$Particles.Initial.MaterialRandom, B=0.65)
         )
         Scale=1
         Bias=0
        )
       )
       Y=1
      )
      B=&Vector2DFromFloat(Value=&FloatFromCurve(FloatCurve=DataInterface() {propertyValues {Curve {keys [<at 0 1.1>, {arriveTangent=-1.1, interpMode=RCIM_Cubic, leaveTangent=-1.1, time=0.8, value=1.1}, <at 1 0 RCIM_Cubic>]}}}))
     )
     B=<1 1>
     Alpha=&FloatFromCurve(FloatCurve=DataInterface() {propertyValues {Curve {keys [<at 0.0574 1>,<at 0.1545 0>]}}})
    ))
   &Drag(
    Drag=&Multiply_Float(
     A=&UniformRangedFloat(Minimum=5, Maximum=8)
     B=&FloatFromCurve(FloatCurve=DataInterface() {propertyValues {Curve {keys [
          {arriveTangent=0.2, interpMode=RCIM_Cubic, time=0.1, value=1}
          {arriveTangent=-1.0144, interpMode=RCIM_Cubic, leaveTangent=-1.0144, tangentMode=RCTM_User, time=1, value=0.2}
         ]}}})
    )
    Use Linear Drag=true
    Use Rotational Drag=true
   )
   &GravityForce(Gravity=&Multiply_VectorByFloat(Vector=&RandomRangeVector(Minimum=<0 0 0>, Maximum=<0 0 -200>), Float=$System.FrontEndScale))
   &SolveForcesAndVelocity(Clamp Velocity=false, Limit Acceleration=false)
  ]
  rendererClasses [&NiagaraSpriteRendererProperties]
  renderers [{rendererClass=&NiagaraSpriteRendererProperties}]
 }
 Toon_Sparks {
  emitterUpdateScript [
   &EmitterState()
   &SpawnBurst_Instantaneous(Spawn Count=12, Spawn Group=1, Use Spawn Probability=false) {module=SpawnBurst_Instantaneous001}
  ]
  particleSpawnScript [
   &InitializeParticle(
    Color=&RandomRangeLinearColor(Minimum=<rgba 0.7969 0.7969 0.7969 0.4>, Maximum=<rgba 0.4479 0.4479 0.4479 0.3>)
    Lifetime=&RandomRangeFloat(Minimum=0.1, Maximum=0.2)
    Sprite Rotation Angle=&RandomRangeFloat(Minimum=0, Maximum=1)
    Sprite Rotation Mode=::NewEnumerator1
    Sprite Size=&Multiply_Vector2D_ByFloat(Float=$System.FrontEndScale, Vector2D=&RandomRangeVector2D(Minimum=<50 33.5>, Maximum=<10 80>))
    Sprite Size Mode=::NewEnumerator4
    UsePositionOffset=true
    Write Material Random=true
   )
   &AddVelocityFromPoint(Module.Use Falloff Distance=false, Velocity Strength=&Multiply_Float(A=200, B=$System.FrontEndScale))
   /SetVariables(Particles.Position=&AddVectorToPosition(Position=$Particles.Position, Vector=$System.FEImpactOffset)) {isSetParameters=true}
  ]
  particleUpdateScript [
   &ParticleState()
   &ScaleSpriteSize(
    Scale Factor=&Vector2DFromCurve(
     Vector2Curve=DataInterface() {propertyValues {XCurve {keys [<at 0 0.778>,<at 1 1>]}, YCurve {keys [<at 0 0.03>,<at 1 1.5>]}}}
     ScaleCurve=<1 1>
    )
    Scale Sprite Size Mode=::NewEnumerator2
   )
   &ScaleColor(
    Scale Alpha=&MergeHelper_FloatSwitch(
     0=&FloatFromCurve(
      FloatCurve=DataInterface() {propertyValues {Curve {keys [
          {arriveTangent=-0.0852, interpMode=RCIM_Cubic, leaveTangent=-0.0852, tangentMode=RCTM_User, time=0, value=1}
          {arriveTangent=-2.6621, interpMode=RCIM_Cubic, leaveTangent=-2.6621, tangentMode=RCTM_User, time=1, value=0}
         ]}}}
      Scale Curve=$System.ToDScalar
     )
     1=&FloatFromCurve(FloatCurve=DataInterface() {propertyValues {Curve {keys [<at 0 1>,<at 1 0>]}}}, Scale Curve=$System.ToDScalar)
     2=?
     3=?
     Index=$Particles.VisibilityTag
     Num=::NewEnumerator0
    )
    Scale RGB=&MergeHelper_VectorSwitch(
     0=&VectorFromFloat(Value=&FloatFromCurve(FloatCurve=DataInterface() {propertyValues {Curve {keys [
           {interpMode=RCIM_Cubic, tangentMode=RCTM_User, time=0, value=1}
           {arriveTangent=-1.4115, interpMode=RCIM_Cubic, leaveTangent=-1.4115, tangentMode=RCTM_User, time=1, value=0}
          ]}}}))
     1=<1 1 1>
     2=?
     3=?
     Index=$Particles.VisibilityTag
     Num=::NewEnumerator0
    )
    ScaleA=true
    ScaleRGB=true
   )
   &CameraOffset(Camera Offset Amount=10)
   &SolveForcesAndVelocity(Clamp Velocity=false, Limit Acceleration=false)
   &ScaleSpriteSizeBySpeed(Max Scale Factor=<1 1>)
  ]
  rendererClasses [&NiagaraSpriteRendererProperties]
  renderers [{rendererClass=&NiagaraSpriteRendererProperties}]
 }
}
summary {
 systemName=NS_Pickaxe_CherryKnot_Impact
 emitters [
  {emitterName=Flash_Elements001, rendererClasses [&NiagaraSpriteRendererProperties]}
  {emitterName=Toon_Sparks, rendererClasses [&NiagaraSpriteRendererProperties]}
  {emitterName=Impact_Smoke, rendererClasses [&NiagaraSpriteRendererProperties]}
  {emitterName=Stickers, rendererClasses [&NiagaraSpriteRendererProperties]}
 ]
 userVariables [
  User.bIsFrontEnd {type=NiagaraBool}
  User.bIsFrontEndPreview {type=NiagaraBool}
  User.PhysicalSurface {type=NiagaraInt32}
  User.SwingSectionIndex {type=NiagaraInt32}
 ]
}
