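# Niagara swing trail for the CherryKnot pickaxe, combining animation-trail ribbons with scattered sticker sprites.
# MainTrail draws two offset-ID ribbons in red and white gradients with a fast alpha fade; WindTrail adds a faint grey ribbon, both using System.TOD for color.
# Stickers spawn per distance from a socket, gated by User.SwingSectionIndex and system age, randomly culled, using six SubUV frames and sine-driven width to fake tumbling.
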NiagaraSystem /BRCosmetics/Weapons/FORT_Melee/Pickaxe_CherryKnot/FX/NS_Pickaxe_CherryKnot_Trail
dependencies {
 usedDataInterfaces [
  &NiagaraDataInterfaceColorCurve
  &NiagaraDataInterfaceCurve
  &NiagaraDataInterfaceSpriteRendererInfo
  &NiagaraDataInterfaceSocketReader
 ]
 usedDynamicInputs []
 usedModules [
  &self:SystemSpawnScript.SystemScriptSource.SystemScriptGraph.NiagaraNodeAssignment_0.SetVariables_#1
  &self:SystemSpawnScript.SystemScriptSource.SystemScriptGraph.NiagaraNodeAssignment_1.SetVariables_#2
  &SystemState
 ]
}
emitters {
 MainTrail {
  emitterSpawnScript [&FN_RibbonIDOffsetUtility()]
  emitterUpdateScript [
   &EmitterState()
   &FN_SpawnAnimTrail(Number of Anim Trails=2)
  ]
  particleSpawnScript [
   &FN_InitializeAnimTrailBoneIndecies(First Bone or Socket Index=2)
   &FN_InitializeAnimTrailBoneIndecies(AnimTrail Selection=1, First Bone or Socket Index=2, Second Bone or Socket Index=0) {module=FN_InitializeAnimTrailBoneIndecies001}
   &FN_InitializeAnimTrail(
    AnimTrailWidth=&MergeHelper_FloatSwitch(0=1, 1=1, 2=?, 3=?, Index=$Particles.SpawnGroup, Num=::NewEnumerator0)
    Color=$System.TOD
   )
   /SetVariables(Particles.RibbonID=/Niagara/DynamicInputs/NiagaraID/MakeNiagaraIDFromInt(
     Index:NiagaraInt32=&Add_Integer(
      A=&ReturnParticleIDAttributes(Particle ID Attribute=::NewEnumerator1, Niagara ID=$Particles.RibbonID)
      B=$Emitter.RibbonIDOffset
     )
     Acquire Tag:NiagaraInt32=0
    )) {isSetParameters=true}
  ]
  particleUpdateScript [
   &FN_UpdateAnimTrail() {module=FN_UpdateAnimTrail001}
   &ScaleColor(
    Scale Alpha=&Multiply_Float(
     A=&FloatFromCurve(
      FloatCurve=DataInterface() {propertyValues {Curve {keys [<at 0 1 RCIM_Cubic>,<at 0.04 0 RCIM_Cubic>]}}}
      CurveIndex=$Particles.Age
      Scale Curve=0.5
     )
     B=&MergeHelper_FloatSwitch(0=1, 1=0.75, 2=?, 3=?, Index=$Particles.SpawnGroup, Num=::NewEnumerator0)
    )
    Scale RGB=&MakeVectorFromLinearColorRGB(Color=&MergeHelper_ColorSwitch(
      0=&ColorFromCurve(ColorCurve=DataInterface() {propertyValues {
         AlphaCurve {keys [<at 0 1 RCIM_Cubic>]}
         BlueCurve {keys [<at 0 0.0729 RCIM_Cubic>,<at 0.4523 0.0153>]}
         GreenCurve {keys [<at 0 0.0757 RCIM_Cubic>,<at 0.4523 0.0122>]}
         RedCurve {keys [<at 0 1 RCIM_Cubic>,<at 0.4523 0.1927>]}
        }})
      1=&ColorFromCurve(ColorCurve=DataInterface() {propertyValues {
         AlphaCurve {keys [<at 0 1 RCIM_Cubic>]}
         BlueCurve {keys [<at 0 1 RCIM_Cubic>,<at 0.4523 0.3594>]}
         GreenCurve {keys [<at 0 1 RCIM_Cubic>,<at 0.4523 0.3594>]}
         RedCurve {keys [<at 0 1 RCIM_Cubic>,<at 0.4523 0.3594>]}
        }})
      2=?
      3=?
      Index=$Particles.SpawnGroup
      Num=::NewEnumerator0
     ))
    ScaleA=true
    ScaleRGB=true
   )
  ]
  rendererClasses [&NiagaraRibbonRendererProperties]
  renderers [{rendererClass=&NiagaraRibbonRendererProperties}]
 }
 Stickers {
  emitterUpdateScript [
   &EmitterState(Life Cycle Mode=::NewEnumerator1, Loop Behavior=::NewEnumerator1, Loop Duration=1, UseLoopDelay=false)
   &SpawnPerUnit(
    Module.Use Max Movement Threshold=false
    Module.Use Movement Tolerance=false
    Spawn Probability:NiagaraFloat=&MergeHelper_FloatSwitch(
     0=&FloatFromCurve(FloatCurve=DataInterface() {propertyValues {Curve {keys [<at 0 0 RCIM_Constant>,<at 0.05 1 RCIM_Constant>]}}}, CurveIndex=$Engine.System.Age)
     1=0
     2=&FloatFromCurve(FloatCurve=DataInterface() {propertyValues {Curve {keys [<at 0.05 1 RCIM_Constant>]}}}, CurveIndex=$Engine.System.Age)
     3=&FloatFromCurve(FloatCurve=DataInterface() {propertyValues {Curve {keys [<at 0.03 1 RCIM_Constant>]}}}, CurveIndex=$Engine.System.Age)
     Index=$User.SwingSectionIndex
     Num=::NewEnumerator2
    )
    Spawn Spacing:NiagaraFloat=&Multiply_Float(A=&RandomRangeFloat(Minimum=20, Maximum=20), B=1)
    Use Spawn Probability=true
   )
  ]
  particleSpawnScript [
   &FN_SetSpawnGroup(Number of Spawn Groups=6)
   &PartitionParticles(Number Of Partitions=6)
   /SetVariables(
    Particles.CustomRandom:NiagaraFloat=&RandomRangeFloat(Minimum=0, Maximum=1)
    Particles.CustomRandomBool:NiagaraBool=&RandomBool()
   ) {isSetParameters=true}
   &InitializeParticle(
    Lifetime=&RandomRangeFloat(Minimum=0.4, Maximum=0.8)
    Sprite Rotation Mode=::NewEnumerator2
    Sprite Size Mode=::NewEnumerator1
    Uniform Sprite Size Max=6
    Uniform Sprite Size Min=2
    UsePositionOffset=false
    Write Material Random=true
   )
   &SocketLocation(
    Interpolate Socket Sampling=true
    SamplePosOffset=true
    Socket Index=&RandomRangeInt(Minimum=0, Maximum=0)
    Socket Reader=$System.Socket Reader
    Socket Sampling Mode=::NewEnumerator3
    VelocitySpeedLimit=false
   )
   /SetVariables(Particles.InitialSize:Vector2f=&Vector2DFromFloat(Value=&RandomRangeFloat(Minimum=50, Maximum=80))) {isSetParameters=true}
   &CameraOffset(Camera Offset Amount=10)
   &ShapeLocation(
    Cylinder Height=50
    Cylinder Height Midpoint=0.2
    Cylinder Radius=5
    MaskSpawnGroup=false
    Offset Mode=::NewEnumerator0
    Override Seed=false
    Shape Primitive=::NewEnumerator1
    Surface Only=false
    UseOwnerScale=true
   )
   &AddVelocity(
    Velocity=&RandomRangeVector(Minimum=<-1 -1 -1>, Maximum=<1 1 1>)
    Velocity Speed Scale=&RandomRangeFloat(Minimum=50, Maximum=100)
   )
   &InheritVelocity(
    Inherited Velocity Amount Scale=&VectorFromFloat(Value=&RandomRangeFloat(Minimum=0.05, Maximum=0.1))
    Speed Limit=true
    UseSpeedThreshold=false
   )
   &SubUVAnimation(
    Direct Frame Index:NiagaraFloat=&Multiply_FloatByInt(Float=1, Integer=$Particles.PartitionParticles.Partition)
    "Sprite Renderer "=DataInterface() {propertyValues {SpriteRenderer=&self:Stickers_0.NiagaraSpriteRendererProperties_2}}
    SubUV Animation Mode:ENiagara_SubUVLookupModeV2=::NewEnumerator5
   )
   &KillParticles(Kill Particles=&RandomBool())
  ]
  particleUpdateScript [
   &Drag(Drag=&RandomRangeFloat(Minimum=5, Maximum=6), Use Linear Drag=true, Use Rotational Drag=false)
   &GravityForce(Gravity=&RandomRangeVector(Minimum=<0 0 0>, Maximum=<0 0 -100>))
   &SolveForcesAndVelocity(Clamp Velocity=false, Limit Acceleration=false)
   &ParticleState()
   &ScaleSpriteSize(Scale Factor=&Lerp_Vector2(
     A=&Multiply_Vector2(
      A=&MakeVector2D(
       X=&Multiply_Float(
        A=1
        B=&Sine(
         Normalized Angle=$Emitter.Age
         Period=&Multiply_Float(
          A=&RandomRangeFloat(Minimum=1.2, Maximum=1.8)
          B=&Add_Float(A=$Particles.Initial.MaterialRandom, B=0.65)
         )
         Scale=1
         Bias=0
        )
       )
       Y=1
      )
      B=&Vector2DFromFloat(Value=&FloatFromCurve(FloatCurve=DataInterface() {propertyValues {Curve {keys [<at 0 1.1>, {arriveTangent=-1.1, interpMode=RCIM_Cubic, leaveTangent=-1.1, time=0.8, value=1.1}, <at 1 0 RCIM_Cubic>]}}}))
     )
     B=<1 1>
     Alpha=&FloatFromCurve(FloatCurve=DataInterface() {propertyValues {Curve {keys [<at 0.0574 1>,<at 0.1545 0>]}}})
    ))
   &SpriteRotationRate(Rotation Rate=&RandomRangeFloat(Minimum=150, Maximum=220))
   &ScaleSpriteSize(Scale Factor=&Vector2DFromFloat(Value=&FloatFromCurve(
      FloatCurve=DataInterface() {propertyValues {Curve {keys [
          <at 0 0 RCIM_Cubic>
          {arriveTangent=-0.0042, interpMode=RCIM_Cubic, leaveTangent=-0.0042, tangentMode=RCTM_User, time=0.05, value=1}
          {arriveTangent=-0.6473, interpMode=RCIM_Cubic, leaveTangent=-0.6473, tangentMode=RCTM_User, time=1, value=0}
         ]}}}
      Scale Curve=2
     ))) {module=ScaleSpriteSize001}
   &ScaleColor(
    Scale RGB=&Multiply_Vector(
     A=<1 1 1>
     B=&VectorFromFloat(Value=&FloatFromCurve(FloatCurve=DataInterface() {propertyValues {Curve {keys [<at 0 0>,<at 0.1 1>,<at 1 1>]}}}))
    )
    ScaleA=true
    ScaleRGB=true
   )
  ]
  rendererClasses [&NiagaraSpriteRendererProperties]
  renderers [{rendererClass=&NiagaraSpriteRendererProperties}]
 }
 WindTrail {
  emitterUpdateScript [
   &EmitterState()
   &FN_SpawnAnimTrail()
  ]
  particleSpawnScript [
   &FN_InitializeAnimTrailBoneIndecies()
   &FN_InitializeAnimTrail(AnimTrailWidth=1.2, Color=<rgba 0.5677 0.5677 0.5677 0.7>, Lifetime=0.2)
  ]
  particleUpdateScript [
   &FN_UpdateAnimTrail() {module=FN_UpdateAnimTrail001}
   &ScaleColor(
    Scale Alpha=&FloatFromCurve(FloatCurve=DataInterface() {propertyValues {Curve {keys [{interpMode=RCIM_Cubic, leaveTangent=-7.0311, tangentMode=RCTM_Break, time=0, value=1}, <at 0.4 0>]}}})
    Scale RGB=&MakeVectorFromLinearColorRGB(Color=$System.TOD)
    ScaleA=true
    ScaleRGB=true
   )
  ]
  rendererClasses [&NiagaraRibbonRendererProperties]
  renderers [{rendererClass=&NiagaraRibbonRendererProperties}]
 }
}
summary {
 systemName=NS_Pickaxe_CherryKnot_Trail
 emitters [
  {emitterName=MainTrail, rendererClasses [&NiagaraRibbonRendererProperties]}
  {emitterName=WindTrail, rendererClasses [&NiagaraRibbonRendererProperties]}
  {emitterName=Stickers, rendererClasses [&NiagaraSpriteRendererProperties]}
 ]
 userVariables [
  User.RibbonIDOffset {type=NiagaraInt32}
  User.SwingSectionIndex {type=NiagaraInt32}
 ]
}
