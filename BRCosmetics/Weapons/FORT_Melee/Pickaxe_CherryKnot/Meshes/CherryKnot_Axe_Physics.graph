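# Physics asset for the CherryKnot pickaxe skeletal mesh, defining a dangling chain on the axe.
# Capsule bodies cover a base_chain anchor bone and five dyn_chain bones.
# Each link is constrained to the one before it, with swing limits widening toward the tip and tight twist limits, so the end swings more freely than the root without spinning.
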PhysicsAsset /BRCosmetics/Weapons/FORT_Melee/Pickaxe_CherryKnot/Meshes/CherryKnot_Axe_Physics
bodies {
 base_chain [{shapeType=Capsule, length=0.01, radius=2.5, center=<-2.6879 0 0>}]
 dyn_chain_01 [{shapeType=Capsule, length=2, radius=1, center=<2 0 0>, rotation=<rot -90 0 0>}]
 dyn_chain_02 [{shapeType=Capsule, length=2, radius=1, center=<2 0 0>, rotation=<rot -90 0 0>}]
 dyn_chain_03 [{shapeType=Capsule, length=2, radius=1, center=<2 0 0>, rotation=<rot -90 0 0>}]
 dyn_chain_04 [{shapeType=Capsule, length=2, radius=1, center=<2 0 0>, rotation=<rot -90 0 0>}]
 dyn_chain_05 [{shapeType=Capsule, length=2, radius=1, center=<2 0 0>, rotation=<rot -90 0 0>}]
}
constraints [
 {bone1Name=dyn_chain_01, bone2Name=base_chain, swing1LimitDegrees=30, swing2LimitDegrees=30, twistLimitDegrees=2.5}
 {bone1Name=dyn_chain_02, bone2Name=dyn_chain_01, swing1LimitDegrees=33, swing2LimitDegrees=33, twistLimitDegrees=2.5}
 {bone1Name=dyn_chain_03, bone2Name=dyn_chain_02, swing1LimitDegrees=40, swing2LimitDegrees=40, twistLimitDegrees=2.5}
 {bone1Name=dyn_chain_04, bone2Name=dyn_chain_03, swing1LimitDegrees=50, swing2LimitDegrees=50, twistLimitDegrees=2.5}
 {bone1Name=dyn_chain_05, bone2Name=dyn_chain_04, swing1LimitDegrees=63, swing2LimitDegrees=63, twistLimitDegrees=2.5}
]
