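# Animation Blueprint for the CherryKnot pickaxe that adds secondary physics motion to its chain and blade bones.
# A front-end flag picks between an AnimDynamics chain and blade stack for menus and a RigidBody plus blade AnimDynamics stack otherwise.
# Bone Driven Controllers then shift base_chain vertically from the first chain link's rotation.
# The update event retunes front-end sim settings by on-pawn state; zipline events set a vector the AnimGraph never reads.
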AnimBlueprint /BRCosmetics/Weapons/FORT_Melee/Pickaxe_CherryKnot/Meshes/CherryKnot_Axe_AnimBP
target_skeleton=&/CherryKnot_Axe_Skeleton
graphs [
 AnimGraph {nodes [
   n1 = "base_chain.translateZ = remap(dyn_chain_01.rotateY) - Bone Driven Controller"!(ComponentPose:Component Space Pose Link Structure=Local To Component(LocalPose=Use cached pose 'PostSimPose'()), Alpha:float=1) -> (Pose:Component Space Pose Link Structure) {
    is=title AnimGraphNode_BoneDrivenController
    bAffectTargetTranslationZ=true
    bUseRange=true
    rangeMax=0
    rangeMin=-90
    remappedMax=0
    remappedMin=0.4
    sourceComponent=RotationY
    sourceBone {boneName=dyn_chain_01}
    targetBone {boneName=base_chain}
   }
   n2 = "base_chain.translateZ = remap(dyn_chain_01.rotateY) - Bone Driven Controller"(ComponentPose=@n1, Alpha=1) {
    bAffectTargetTranslationZ=true
    bUseRange=true
    rangeMax=90
    rangeMin=0
    remappedMax=0.4
    sourceComponent=RotationY
    sourceBone {boneName=dyn_chain_01}
    targetBone {boneName=base_chain}
   }
   Output Pose(Result=Component To Local(ComponentPose=@n2)) {layerGroup=None}
   n3 = RigidBody!(ComponentPose=Local To Component(LocalPose=Use cached pose 'StartingPose'()), Alpha, OverrideWorldGravity:Vector="0,0,0", ExternalForce) -> (Pose) {
    bOverrideWorldGravity=true
    overridePhysicsAsset=&/CherryKnot_Axe_Physics
    simulationSpace=BaseBoneSpace
    baseBoneRef {boneName=base_chain}
    simSpaceSettings {
     maxAngularAcceleration=175
     maxAngularVelocity=55
     maxLinearAcceleration=1175
     maxLinearVelocity=550
     worldAlpha=1
     externalLinearDragV=<0.2 0.2 0.2>
    }
   }
   n4 = Anim Dynamics - Bone: dyn_blade01_cr!(
    ComponentPose=@n3
    Alpha=1
    ComponentLinearAccScale:Vector="0.4,0.4,0.4"
    ComponentLinearVelScale:Vector="0.4,0.4,0.4"
    ExternalForce="0,0,0"
   ) -> (Pose) {
    angularDampingOverride=1
    angularSpringConstant=10500
    bAngularSpring=true
    bOverrideAngularDamping=true
    bUseGravityOverride=true
    gravityScale=0.35
    simulationSpace=RootRelative
    boundBone {boneName=dyn_blade01_cr}
    componentAppliedLinearAccClamp=<30000 30000 30000>
    componentLinearAccScale=<0.4 0.4 0.4>
    componentLinearVelScale=<0.4 0.4 0.4>
    gravityOverride {z=-180}
    physicsBodyDefinitions [{
      boundBone {boneName=dyn_blade01_cr}
      boxExtents=<40 45 45>
      constraintSetup {angularLimitsMax=<0 2.5 2.5>, angularLimitsMin=<0 -2.5 -2.5>, angularTarget=<1 0 0>}
      localJointOffset=<1.5 0 0>
     }]
    simSpaceSettings {maxAngularAcceleration=8, maxAngularVelocity=6, simSpaceAngularAlpha=0.6}
   }
   n5 = Anim Dynamics - Bone: dyn_blade02_cr!(
    ComponentPose=@n4
    Alpha=1
    ComponentLinearAccScale:Vector="0.4,0.4,0.4"
    ComponentLinearVelScale:Vector="0.4,0.4,0.4"
    ExternalForce="0,0,0"
   ) -> (Pose) {
    angularDampingOverride=1
    angularSpringConstant=10500
    bAngularSpring=true
    bOverrideAngularDamping=true
    bUseGravityOverride=true
    gravityScale=0.35
    simulationSpace=RootRelative
    boundBone {boneName=dyn_blade02_cr}
    componentAppliedLinearAccClamp=<30000 30000 30000>
    componentLinearAccScale=<0.4 0.4 0.4>
    componentLinearVelScale=<0.4 0.4 0.4>
    gravityOverride {z=-180}
    physicsBodyDefinitions [{
      boundBone {boneName=dyn_blade02_cr}
      boxExtents=<35 40 40>
      constraintSetup {angularLimitsMax=<0 4 4>, angularLimitsMin=<0 -4 -4>, angularTarget=<1 0 0>}
      localJointOffset=<1.5 0 0>
     }]
    simSpaceSettings {maxAngularAcceleration=8, maxAngularVelocity=6, simSpaceAngularAlpha=0.6}
   }
   n6 = Anim Dynamics - Bone: dyn_blade03_cr!(
    ComponentPose=@n5
    Alpha=1
    ComponentLinearAccScale:Vector="0.4,0.4,0.4"
    ComponentLinearVelScale:Vector="0.4,0.4,0.4"
    ExternalForce="0,0,0"
   ) -> (Pose) {
    angularDampingOverride=1
    angularSpringConstant=10500
    bAngularSpring=true
    bOverrideAngularDamping=true
    bUseGravityOverride=true
    gravityScale=0.35
    simulationSpace=RootRelative
    boundBone {boneName=dyn_blade03_cr}
    componentAppliedLinearAccClamp=<30000 30000 30000>
    componentLinearAccScale=<0.4 0.4 0.4>
    componentLinearVelScale=<0.4 0.4 0.4>
    gravityOverride {z=-180}
    physicsBodyDefinitions [{
      boundBone {boneName=dyn_blade03_cr}
      boxExtents=<35 40 40>
      constraintSetup {angularLimitsMax=<0 6 6>, angularLimitsMin=<0 -6 -6>, angularTarget=<1 0 0>}
      localJointOffset=<1.5 0 0>
     }]
    simSpaceSettings {maxAngularAcceleration=8, maxAngularVelocity=6, simSpaceAngularAlpha=0.6}
   }
   n7 = Use cached pose 'StartingPose'()
   n8 = Get SimSetting() -> (SimSetting:Anim Phys Sim Space Settings Structure)
   n9 = Anim Dynamics - Bone: dyn_chain_01!(
    ComponentPose="Transform (Modify) Bone - Bone: base_chain"(
     ComponentPose=Local To Component(LocalPose=@n7)
     Alpha=Get OnPawnAlpha() -> (OnPawnAlpha:double)
     Translation="0,0,0"
     Rotation="0,70,0"
     Scale="1,1,1"
    ) {rotationMode=BMM_Additive, rotationSpace=BCS_BoneSpace, boneToModify {boneName=base_chain}, rotation {yaw=70}}
    Alpha=1
    SimSpaceSettings:Anim Phys Sim Space Settings Structure=@n8
    ExternalForce="0,0,0"
   ) -> (Pose) {
    angularDampingOverride=0.993
    bAngularSpring=true
    bOverrideAngularDamping=true
    bUseGravityOverride=true
    linearDampingOverride=0.99
    linearSpringConstant=550
    simulationSpace=RootRelative
    boundBone {boneName=dyn_chain_01}
    componentAppliedLinearAccClamp=<3000 3000 6000>
    componentLinearAccScale=<0.25 0.25 0.25>
    componentLinearVelScale=<0.2 0.2 0.1>
    gravityOverride {z=-980}
    physicsBodyDefinitions [{
      boundBone {boneName=dyn_chain_01}
      boxExtents=<9.5 9.5 9.5>
      constraintSetup {angularLimitsMax=<5 40 20>, angularLimitsMin=<-5 -90 -20>, angularTarget=<1 0 0>}
      localJointOffset=<1 0 0>
     }]
    simSpaceSettings {maxAngularAcceleration=15, maxAngularVelocity=8, simSpaceAngularAlpha=0.3}
   }
   n10 = Anim Dynamics - Bone: dyn_chain_02!(ComponentPose=@n9, Alpha=1, SimSpaceSettings:Anim Phys Sim Space Settings Structure=@n8, ExternalForce="0,0,0") -> (Pose) {
    angularDampingOverride=0.993
    bAngularSpring=true
    bOverrideAngularDamping=true
    bUseGravityOverride=true
    linearDampingOverride=0.99
    linearSpringConstant=550
    simulationSpace=RootRelative
    boundBone {boneName=dyn_chain_02}
    componentAppliedLinearAccClamp=<3000 3000 6000>
    componentLinearAccScale=<0.25 0.25 0.25>
    componentLinearVelScale=<0.2 0.2 0.1>
    gravityOverride {z=-980}
    physicsBodyDefinitions [{
      boundBone {boneName=dyn_chain_02}
      boxExtents=<9.5 9.5 9.5>
      constraintSetup {angularLimitsMax=<5 40 20>, angularLimitsMin=<-5 -90 -20>, angularTarget=<1 0 0>}
      localJointOffset=<1 0 0>
     }]
    simSpaceSettings {maxAngularAcceleration=15, maxAngularVelocity=8, simSpaceAngularAlpha=0.3}
   }
   n11 = Anim Dynamics - Bone: dyn_chain_03!(ComponentPose=@n10, Alpha=1, SimSpaceSettings:Anim Phys Sim Space Settings Structure=@n8, ExternalForce="0,0,0") -> (Pose) {
    angularDampingOverride=0.993
    bAngularSpring=true
    bOverrideAngularDamping=true
    bUseGravityOverride=true
    linearDampingOverride=0.99
    linearSpringConstant=550
    simulationSpace=RootRelative
    boundBone {boneName=dyn_chain_03}
    componentAppliedLinearAccClamp=<3000 3000 6000>
    componentLinearAccScale=<0.25 0.25 0.25>
    componentLinearVelScale=<0.2 0.2 0.1>
    gravityOverride {z=-980}
    physicsBodyDefinitions [{
      boundBone {boneName=dyn_chain_03}
      boxExtents=<9.5 9.5 9.5>
      constraintSetup {angularLimitsMax=<5 20 30>, angularLimitsMin=<-5 -20 -30>, angularTarget=<1 0 0>}
      localJointOffset=<1 0 0>
     }]
    simSpaceSettings {maxAngularAcceleration=15, maxAngularVelocity=8, simSpaceAngularAlpha=0.3}
   }
   n12 = Anim Dynamics - Bone: dyn_chain_04!(ComponentPose=@n11, Alpha=1, SimSpaceSettings:Anim Phys Sim Space Settings Structure=@n8, ExternalForce="0,0,0") -> (Pose) {
    angularDampingOverride=0.993
    bAngularSpring=true
    bOverrideAngularDamping=true
    bUseGravityOverride=true
    linearDampingOverride=0.99
    linearSpringConstant=550
    simulationSpace=RootRelative
    boundBone {boneName=dyn_chain_04}
    componentAppliedLinearAccClamp=<3000 3000 6000>
    componentLinearAccScale=<0.25 0.25 0.25>
    componentLinearVelScale=<0.2 0.2 0.1>
    gravityOverride {z=-980}
    physicsBodyDefinitions [{
      boundBone {boneName=dyn_chain_04}
      boxExtents=<9 9 9>
      constraintSetup {angularLimitsMax=<5 20 30>, angularLimitsMin=<-5 -20 -30>, angularTarget=<1 0 0>}
      localJointOffset=<1 0 0>
     }]
    simSpaceSettings {maxAngularAcceleration=15, maxAngularVelocity=8, simSpaceAngularAlpha=0.3}
   }
   n13 = Anim Dynamics - Bone: dyn_chain_05!(ComponentPose=@n12, Alpha=1, SimSpaceSettings:Anim Phys Sim Space Settings Structure=@n8, ExternalForce="0,0,0") -> (Pose) {
    angularDampingOverride=0.993
    bAngularSpring=true
    bOverrideAngularDamping=true
    bUseGravityOverride=true
    linearDampingOverride=0.99
    linearSpringConstant=550
    simulationSpace=RootRelative
    boundBone {boneName=dyn_chain_05}
    componentAppliedLinearAccClamp=<3000 3000 6000>
    componentLinearAccScale=<0.25 0.25 0.25>
    componentLinearVelScale=<0.2 0.2 0.1>
    gravityOverride {z=-980}
    physicsBodyDefinitions [{
      boundBone {boneName=dyn_chain_05}
      boxExtents=<8 8 8>
      constraintSetup {angularLimitsMax=<5 20 30>, angularLimitsMin=<-5 -20 -30>, angularTarget=<1 0 0>}
      localJointOffset=<1 0 0>
     }]
    simSpaceSettings {maxAngularAcceleration=15, maxAngularVelocity=8, simSpaceAngularAlpha=0.3}
   }
   n14 = Get FEAnimDynAlpha() -> (FEAnimDynAlpha:double)
   n15 = Anim Dynamics - Bone: dyn_blade01_cr!(ComponentPose=@n13, Alpha=@n14, ExternalForce="0,0,0") -> (Pose) {
    angularDampingOverride=1
    angularSpringConstant=19000
    bAngularSpring=true
    bOverrideAngularDamping=true
    bUseGravityOverride=true
    gravityScale=0.35
    simulationSpace=RootRelative
    boundBone {boneName=dyn_blade01_cr}
    componentAppliedLinearAccClamp=<40000 40000 40000>
    componentLinearAccScale=<0.4 0.4 0.4>
    componentLinearVelScale=<0.4 0.4 0.4>
    gravityOverride {z=-180}
    physicsBodyDefinitions [{
      boundBone {boneName=dyn_blade01_cr}
      boxExtents=<38 43 43>
      constraintSetup {angularLimitsMax=<0 2.5 2.5>, angularLimitsMin=<0 -2.5 -2.5>, angularTarget=<1 0 0>}
      localJointOffset=<0.5 0 0>
     }]
    simSpaceSettings {maxAngularAcceleration=8, maxAngularVelocity=6, simSpaceAngularAlpha=0.4}
   }
   n16 = Anim Dynamics - Bone: dyn_blade02_cr!(ComponentPose=@n15, Alpha=@n14, ExternalForce="0,0,0") -> (Pose) {
    angularDampingOverride=1
    angularSpringConstant=19000
    bAngularSpring=true
    bOverrideAngularDamping=true
    bUseGravityOverride=true
    gravityScale=0.35
    simulationSpace=RootRelative
    boundBone {boneName=dyn_blade02_cr}
    componentAppliedLinearAccClamp=<40000 40000 40000>
    componentLinearAccScale=<0.4 0.4 0.4>
    componentLinearVelScale=<0.4 0.4 0.4>
    gravityOverride {z=-180}
    physicsBodyDefinitions [{
      boundBone {boneName=dyn_blade02_cr}
      boxExtents=<38 43 43>
      constraintSetup {angularLimitsMax=<0 4 4>, angularLimitsMin=<0 -4 -4>, angularTarget=<1 0 0>}
      localJointOffset=<0.5 0 0>
     }]
    simSpaceSettings {maxAngularAcceleration=8, maxAngularVelocity=6, simSpaceAngularAlpha=0.4}
   }
   n17 = Anim Dynamics - Bone: dyn_blade03_cr!(ComponentPose=@n16, Alpha=@n14, ExternalForce="0,0,0") -> (Pose) {
    angularDampingOverride=1
    angularSpringConstant=19000
    bAngularSpring=true
    bOverrideAngularDamping=true
    bUseGravityOverride=true
    gravityScale=0.35
    simulationSpace=RootRelative
    boundBone {boneName=dyn_blade03_cr}
    componentAppliedLinearAccClamp=<40000 40000 40000>
    componentLinearAccScale=<0.4 0.4 0.4>
    componentLinearVelScale=<0.4 0.4 0.4>
    gravityOverride {z=-180}
    physicsBodyDefinitions [{
      boundBone {boneName=dyn_blade03_cr}
      boxExtents=<38 43 43>
      constraintSetup {angularLimitsMax=<0 6 6>, angularLimitsMin=<0 -6 -6>, angularTarget=<1 0 0>}
      localJointOffset=<0.5 0 0>
     }]
    simSpaceSettings {maxAngularAcceleration=8, maxAngularVelocity=6, simSpaceAngularAlpha=0.4}
   }
   Save cached pose 'PostSimPose'(Pose=Blend Poses by bool(
     BlendPose_0=Component To Local(ComponentPose=@n17)
     BlendPose_1=Component To Local(ComponentPose=@n6)
     BlendTime_0=0
     BlendTime_1=0
     bActiveValue=Get bIsInFrontEnd()
    )) {cacheName=PostSimPose}
   AnimGraphNode_SubInput_0 = Input Pose()
   Save cached pose 'StartingPose'(Pose=@AnimGraphNode_SubInput_0) {cacheName=StartingPose}
  ]}
 EventGraph {nodes [
   Event BlueprintUpdateAnimation()
   > n1 = Branch(Condition=AND Boolean(A=Get bIsOnPawn() -> (bIsOnPawn:bool), B=Get bIsInFrontEnd()))
   > Set FEAnimDynAlpha(FEAnimDynAlpha:double=0.7) -> (Output_Get:double)
   > Set OnPawnAlpha(OnPawnAlpha:double=1) -> (Output_Get:double)
   > Set OffPawnAlpha(OffPawnAlpha:double=0) -> (Output_Get:double)
   > Set SimSetting(SimSetting:Anim Phys Sim Space Settings Structure=Get OnPawn() -> (OnPawn:Anim Phys Sim Space Settings Structure)) -> (Output_Get:Anim Phys Sim Space Settings Structure)
   Set FEAnimDynAlpha(execute=@n1.else, FEAnimDynAlpha=0.85)
   > Set OnPawnAlpha(OnPawnAlpha=0)
   > Set OffPawnAlpha(OffPawnAlpha=1)
   > Set SimSetting(SimSetting=Get OffPawn() -> (OffPawn:Anim Phys Sim Space Settings Structure))
   n2 = Event BlueprintProcessAnimInputEvent()
   n3 = IsNamed(self=@n2.AnimInputEvent, InEventName=aie.Player.Zipline.Entered)
   n4 = OR Boolean(A=Get bIsMoving2D(), B=Get bIsTacticalSprinting())
   n5 = Get bIsJumping()
   > n6 = Branch(Condition=OR Boolean(A=@n3, B=AND Boolean!(A=@n3, B=@n5, C:bool=@n4) -> (ReturnValue)))
   > Set ZiplineAnimDyn(ZiplineAnimDyn:Vector="0, 0, 0") -> (Output_Get:Vector)
   Set ZiplineAnimDyn(execute=[@n6.else, @n7.then], ZiplineAnimDyn="0.25,0.25,0.25")
   n7 = Branch(Condition=IsNamed(self=@n2.AnimInputEvent, InEventName=aie.Player.Zipline.Exited))
   TryGetPawnOwner()
  ]}
]
variables [
 FEAnimDynAlpha {type=number, default=1}
 Gravity Override {type=Vector, default=<0 0 -980>}
 OffPawn {
  type=AnimPhysSimSpaceSettings
  default {maxAngularAcceleration=15, maxAngularVelocity=8, simSpaceAngularAlpha=1, externalAngularVelocity=<0.25 0 0>}
 }
 OffPawnAlpha {type=number}
 OnPawn {
  type=AnimPhysSimSpaceSettings
  default {maxAngularAcceleration=15, maxAngularVelocity=8, simSpaceAngularAlpha=0.3, externalAngularVelocity=<0 0 0>}
 }
 OnPawnAlpha {type=number}
 SimSetting {
  type=AnimPhysSimSpaceSettings
  default {maxAngularAcceleration=15, maxAngularVelocity=8, simSpaceAngularAlpha=0.3, externalAngularVelocity=<0 0 0>}
 }
 ZiplineAnimDyn {type=Vector, default=<0.25 0.25 0.25>}
]
